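# Translucent unlit ribbon material for the harsh trail effect of the RecordScratch pickaxe, built for Niagara ribbons.
# Sphere masks with zeroed U tiling shape soft bands across the width, broken up by two panning inky-smoke samples on the second UV set.
# A noise-warped corner mask and UV gradients shade the streak and fade both ends.
# Particle color and alpha tint the result, which feeds an emissive adjustment function and a depth-faded mask.
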Material /BRCosmetics/Weapons/FORT_Melee/Pickaxe_RecordScratch/FX/Materials/M_RecordScratch_Harsh_Trail
graph {
 properties {bUsedWithNiagaraRibbons=true, blendMode=BLEND_Translucent, pixelDepthOffsetMode=PDOM_Legacy, shadingModel=MSM_Unlit, usedShadingModels=Unlit}
 nodes [
  n1 = SphereMask(A=TextureCoordinate() {uTiling=0}, B=Constant2Vector() {g=0.5}, Radius=Constant() {r=0.475}, Hardness=Constant() {r=0.9})
  n2 = SphereMask(A=TextureCoordinate() {uTiling=0}, B=Constant2Vector() {g=0.5}, Radius=Constant() {r=0.45}, Hardness=Constant() {r=0.5})
  n3 = Saturate(None=Subtract(A=OneMinus(None=@n2)) {constA=0.5, constB=0.6})
  n4 = TextureSample(UVs=Panner(Coordinate=TextureCoordinate() {coordinateIndex=1, uTiling=0.1, vTiling=0.2}) {speedX=-1, speedY=-0.2}) {texture=/Game/Effects/Fort_Effects/Textures/Tile/T_Inky_Smoke_Tile2}
  n5 = Multiply(A=Add(A=@n4.R, B=@n3), B=@n3)
  n6 = Power(Base=Multiply(A=Multiply(A=@n1, B=@n5)) {constB=6})
  n7 = TextureCoordinate()
  n8 = ComponentMask(None=@n7) {g=true}
  n9 = Multiply(A=Add(A=OneMinus(None=@n8)) {constB=0}) {constB=0.2}
  n10 = TextureCoordinate() {coordinateIndex=1, uTiling=0.1, vTiling=5}
  n11 = Multiply(A=TextureSample(UVs=Panner(Coordinate=@n10) {speedX=-0.5, speedY=0.5}) {texture=&T_Inky_Smoke_Tile}.R) {constB=0.5}
  n12 = Subtract(A=SphereMask(A=TextureCoordinate() {uTiling=0}, B=Constant2Vector() {g=0.5}, Radius=Constant() {r=0.5}, Hardness=Constant())) {constA=0.5, constB=0.15}
  n13 = Multiply(A=Multiply(A=Add(A=@n4.R, B=@n11), B=Saturate(None=@n12))) {constB=0.5}
  n14 = Add(A=Multiply(A=Multiply(A=@n6, B=@n9)), B=@n13)
  n15 = ComponentMask(None=@n7) {r=true}
  n16 = Multiply(A=@n1, B=OneMinus(None=@n15))
  n17 = Multiply(A=@n16) {constB=0.5}
  n18 = TextureCoordinate()
  n19 = Constant2Vector() {g=1, r=0.9}
  n20 = Add(A=@n13, B=Constant() {r=1})
  n21 = Add(A=@n13, B=Constant() {r=0.5})
  n22 = Power(Base=OneMinus(None=SphereMask(A=@n18, B=@n19, Radius=@n20, Hardness=@n21))) {constExponent=5}
  n23 = Multiply(A=Multiply(A=@n16, B=Saturate(None=@n22)))
  n24 = Multiply(A=Add(A=Add(A=@n14, B=@n17), B=@n23), B=@n16)
  n25 = Saturate(None=Multiply(A=@n15) {constB=7.5})
  n26 = ParticleColor()
  n27 = Multiply(A=Multiply(A=Saturate(None=@n24), B=@n25), B=@n26.A)
  MaterialFunctionCall(Emissive=Multiply(A=@n27, B=@n26.RGB)) {materialFunction=&MF_FX_EmissiveAdjustments}
  n28 = Add(A=Add(A=@n14, B=Multiply(A=@n16) {constB=0.1}), B=@n23)
  Multiply(A=Saturate(None=Multiply(A=@n28, B=@n16)), B=@n25)
  Saturate(None=Multiply(A=@n27, B=DepthFade() {fadeDistanceDefault=50}))
 ]
}
