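# Translucent unlit Niagara sprite material for the RecordScratch pickaxe's splash effect, using a hand-painted dissolve texture.
# The texture's green channel, multiplied by particle color or remapped between two tint colors via a static switch, feeds emissive adjustments.
# The same channel drives a smoothstep erosion, with inverted particle alpha as threshold and a dynamic parameter setting softness.
# Depth fade softens the eroded opacity except at low quality.
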Material /BRCosmetics/Weapons/FORT_Melee/Pickaxe_RecordScratch/FX/Materials/M_Splash_RecordScratch
graph {
 properties {bUsedWithNiagaraSprites=true, blendMode=BLEND_Translucent, pixelDepthOffsetMode=PDOM_Legacy, shadingModel=MSM_Unlit, usedShadingModels=Unlit}
 nodes [
  n1 = TextureSampleParameter2D() {parameterName=Tex, texture=/BRCosmetics/Weapons/FORT_Melee/Pickaxe_RecordScratch/FX/Textures/T_Water_Splash_Handpainted_Dissolve}
  n2 = ParticleColor()
  n3 = LinearInterpolate(
   A=VectorParameter() {parameterName=ColorA, defaultValue=<rgba 0.1477 0.2751 0.2865 1>}.RGB
   B=VectorParameter() {parameterName=ColorB, defaultValue=<rgba 0.5406 0.9701 1 1>}.RGB
   Alpha=@n1.G
  )
  n4 = ParticleColor()
  MaterialFunctionCall(Emissive=StaticSwitchParameter(True=Multiply(A=@n1.G, B=@n2.RGB), False=Multiply(A=@n3, B=@n4.RGB)) {defaultValue=true, parameterName=UseParticle Color}) {materialFunction=&MF_FX_EmissiveAdjustments}
  n5 = OneMinus(None=@n4.A)
  n6 = MaterialFunctionCall(
   Alpha=@n1.G
   Min=@n5
   Max=Add(
    A=@n5
    B=DynamicParameter() {defaultValue=<rgba 0.1 1 1 1>, paramNames [SmoothstepSoftness, Random, WarpStrength, Panning Speed]}.SmoothstepSoftness
   )
  ) {materialFunction=&SmoothStep}
  n7 = DepthFade(Opacity=@n6.Result, FadeDistance=ScalarParameter() {defaultValue=50, parameterName=DepthFade}) {fadeDistanceDefault=50}
  QualitySwitch(Default=@n7, Low=@n6.Result, High=@n7, Medium=@n7, Epic=@n7)
 ]
}
