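# Master cloth-model material giving an anime-shaded liquid look, stored with the RecordScratch pickaxe cosmetic.
# A panning distortion normal, flattened, feeds MF_AnimeShading between the shared character PreFX and PostFX functions.
# Its emissive blends toward a stepped Fresnel rim tinted by time of day and modulated by shading, then gets multiplied by a rerouted value, apparently the water Fresnel.
# Shared visor, WPO, flipbook mouth, cloth sim and SkinFX functions follow.
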Material /BRCosmetics/Weapons/FORT_Melee/Pickaxe_RecordScratch/Materials/M_MASTER_Liquid_AnimeShading
graph {
 properties {
  bAutomaticallySetUsageInEditor=false
  bUseMaterialAttributes=true
  bUsedWithClothing=true
  bUsedWithMeshDeformer=true
  bUsedWithSkeletalMesh=true
  pixelDepthOffsetMode=PDOM_Legacy
  shadingModel=MSM_Cloth
  usedShadingModels=Cloth
 }
 nodes [
  MaterialFunctionCall!(Material) {materialFunction=/Game/Packages/Fortress_SharedMaterials/Anime_Master_Material/MF_ToonOutline_Mobile}
  n1 = MaterialFunctionCall() {materialFunction=&BaseFN_CharacterFunction_PreFX}
  n2 = TextureCoordinate()
  n3 = ScalarParameter() {defaultValue=3, group=CustomToon, parameterName=NormalTile}
  n4 = ScalarParameter() {group=CustomToon, parameterName=SpeedX}
  n5 = ScalarParameter() {group=CustomToon, parameterName=SpeedY}
  n6 = TextureSample(UVs=Panner(Coordinate=Multiply(A=@n2, B=@n3), Speed=AppendVector(A=@n4, B=@n5))) {samplerType=SAMPLERTYPE_Normal, texture=&T_Weap_Distort_N_02}
  n7 = ScalarParameter() {defaultValue=0.2, group=CustomToon, parameterName=NormalFaltness}
  n8 = MaterialFunctionCall!(
   Material=SetMaterialAttributes!(
    MaterialAttributes=@n1.Result
    Normal=MaterialFunctionCall(Normal=@n6.RGB, Flatness=@n7) {materialFunction=&FlattenNormal}.Result
   ) {attributeSetTypes [#1]}
   LitDiffuseOverride=VectorParameter() {group=CustomToon, parameterName=LitDiffuseOverride, defaultValue=<rgba 1 1 1 1>}.RGB
   SSC_AlphaOverride
   Tint Mask
   DF InkLine Visibility Mask
   DF Inkline Injection
   DF Inkline Color Override
   DF Inkline Color Override Mask
   CustomUVs
  ) {materialFunction=/Game/Packages/Fortress_SharedMaterials/Anime_Master_Material/MF_AnimeShading}
  n9 = GetMaterialAttributes(None=@n8.Result) {attributeGetTypes [#2, #1]}
  n10 = VectorParameter() {group=CustomToon, parameterName=FresnelOutline, defaultValue=<rgba 1 1 1 1>}
  n11 = VectorParameter() {group=CustomToon, parameterName=FresnelOutlineNight, defaultValue=<rgba 1 1 1 1>}
  n12 = Step(
   Y=ScalarParameter() {defaultValue=0.5, group=CustomToon, parameterName=FRE_Step}
   X=Fresnel(
    ExponentIn=ScalarParameter() {defaultValue=2, group=CustomToon, parameterName=FRE_ExponentIn}
    BaseReflectFractionIn=ScalarParameter() {defaultValue=0.04, group=CustomToon, parameterName=FRE_ReflectFraction}
    Normal=Transform(None=@n9.Normal)
   )
  ) {constY=0.5}
  n13 = Multiply(A=@n12) {constB=0.2}
  n14 = ComponentMask(None=@n8.Shading) {r=true}
  n15 = Multiply(
   A=LinearInterpolate(
    A=@n9.Emissive Color
    B=MaterialFunctionCall(Color_Morning=@n10.RGB, Color_Day=@n10.RGB, Color_Evening=@n11.RGB, Color_Night=@n11.RGB) {materialFunction=&Color_Time_of_Day}.Result
    Alpha=LinearInterpolate(A=@n13, B=@n12, Alpha=@n14)
   )
   B=NamedRerouteUsage()
  )
  n16 = MaterialFunctionCall!(In=MaterialFunctionCall!(In=SetMaterialAttributes!(MaterialAttributes=@n9.MaterialAttributes, Emissive Color=@n15) {attributeSetTypes [#2]}) {materialFunction=/Game/Packages/Fortress_SharedMaterials/Anime_Master_Material/MF_AnimatedVisor}.Result) {materialFunction=/Game/Characters/Player/Female/Medium/Bodies/F_MED_Peppermint/Materials/MF_HairPuffWPO}
  n17 = MaterialFunctionCall!(In=MaterialFunctionCall!(In=MaterialFunctionCall!(In=@n16.Result) {materialFunction=/Game/Packages/Fortress_SharedMaterials/Anime_Master_Material/MF_BobbingWPO}.Result) {materialFunction=/Game/Packages/Fortress_SharedMaterials/Anime_Master_Material/MF_PickupShrinking}.Result) {materialFunction=/Game/Packages/Fortress_SharedMaterials/Anime_Master_Material/MF_ToonFlipbookMouth}
  MaterialFunctionCall(
   AO=@n1.AO
   Distance Blend=@n1.AO
   Fresnel=@n1.AO
   Gmap=@n1.AO
   Pre Skinned Local Position=@n1.AO
   Material=MaterialFunctionCall!(
    MA=MaterialFunctionCall!(Material=@n17.Result, SkinFX_Mask=@n8.FX) {materialFunction=/Game/Global/MaterialFunctions/MF_ClothSimWPO_v2}.Result
    Shading=@n8.Shading
    InkLine Mask=@n8.InkLine Mask
    SSC=@n8.SSC
    SkinFX=@n8.FX
   ) {materialFunction=/Game/Packages/Fortress_SharedMaterials/Anime_Master_Material/MF_AnimeSkinFX}.MA
  ) {materialFunction=&BaseFN_CharacterFunction_PostFX}
  n18 = Saturate(None=Multiply(A=Fresnel() {baseReflectFraction=0.05, exponent=1}, B=ScalarParameter() {defaultValue=5, group=CustomToon, parameterName=WaterThreshold}) {constB=3})
  n19 = ScalarParameter() {defaultValue=3, group=CustomToon, parameterName=WaterPow}
  NamedRerouteDeclaration(None=LinearInterpolate(Alpha=Power(Base=@n18, Exp=@n19)) {constA=1, constB=1.5}) {name=CachedWaterFresnel, nodeColor=<rgba 0.9688 1 0 1>}
  TextureSample() {samplerType=SAMPLERTYPE_Normal, texture=/Game/Global/Textures/Default/Flowmaps/T_Smoke_FL_01}
 ]
}
