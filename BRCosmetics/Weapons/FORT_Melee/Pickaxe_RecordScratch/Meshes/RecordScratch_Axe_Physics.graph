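# Physics asset for the RecordScratch two-handed pickaxe, giving capsule bodies and joints to its handle, drip pieces and hanging pendant.
# Three small base bodies attach to the handle with free swing and twist, each rooting a chain of dyn bodies.
# The main drip chain forks into mid and top branches, while bottom drip and pendant form separate chains.
# Drip joints allow only a few degrees of motion, whereas pendant links swing much wider with tight twist.
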PhysicsAsset /BRCosmetics/Weapons/FORT_Melee/Pickaxe_RecordScratch/Meshes/RecordScratch_Axe_Physics
bodies {
 Melee_TwoHanded_Handle [{
   shapeType=Capsule
   length=84.1968
   radius=7.3293
   shapeName=Melee_TwoHanded_Handle_capsule
   center=<1.0051 0.6335 17.7095>
   rotation=<rot 4.169 -90 -0.0884>
  }]
 base_drip_bot_00 [{shapeType=Capsule, length=0.505, radius=0.505, shapeName=base_drip_bot_00_capsule, rotation=<rot 0 -90 0>}]
 base_drip_main_00 [{shapeType=Capsule, length=0.505, radius=0.505, shapeName=base_drip_main_00_capsule, rotation=<rot 0 -90 0>}]
 base_pendant_00 [{shapeType=Capsule, length=0.505, radius=0.505, shapeName=base_pendant_00_capsule, rotation=<rot 0 -90 0>}]
 dyn_drip_bot_00 [{shapeType=Capsule, length=2.4733, radius=1.6135, shapeName=dyn_drip_bot_00_capsule, center=<1.2356 0 0>, rotation=<rot 90 0 0>}]
 dyn_drip_bot_01 [{shapeType=Capsule, length=3.0206, radius=1.3485, shapeName=dyn_drip_bot_01_capsule, center=<1.7653 0 0>, rotation=<rot 90 0 0>}]
 dyn_drip_main_00 [{shapeType=Capsule, length=4.0113, radius=3.7956, shapeName=dyn_drip_main_00_capsule, center=<4.6805 0 0.63>, rotation=<rot 90 0 0>}]
 dyn_drip_main_01 [{shapeType=Capsule, length=2.6649, radius=3.3556, shapeName=dyn_drip_main_01_capsule, center=<3.4581 0 -0.1152>, rotation=<rot 90 0 0>}]
 dyn_drip_mid_00 [{shapeType=Capsule, length=1.5599, radius=0.8709, shapeName=dyn_drip_mid_00_capsule, center=<0.6518 0 0>, rotation=<rot 90 0 0>}]
 dyn_drip_mid_01 [{shapeType=Capsule, length=1.6103, radius=0.9145, shapeName=dyn_drip_mid_01_capsule, center=<1.8078 0 0>, rotation=<rot 90 0 0>}]
 dyn_drip_top_00 [{shapeType=Capsule, length=3.6993, radius=1.9879, shapeName=dyn_drip_top_00_capsule, center=<2.2512 0 -0.1476>, rotation=<rot 90 0 0>}]
 dyn_drip_top_01 [{shapeType=Capsule, length=2.173, radius=1.1711, shapeName=dyn_drip_top_01_capsule, center=<0.9841 0 0>, rotation=<rot 90 0 0>}]
 dyn_drip_top_02 [{shapeType=Capsule, length=2, radius=1, shapeName=dyn_drip_top_02_capsule, center=<0.8042 0 0>, rotation=<rot 90 0 0>}]
 dyn_pendant_00 [{shapeType=Capsule, length=2, radius=1, shapeName=dyn_pendant_00_capsule, center=<2 0 0>, rotation=<rot 90 0 0>}]
 dyn_pendant_01 [{shapeType=Capsule, length=2, radius=1, shapeName=dyn_pendant_01_capsule, center=<2 0 0>, rotation=<rot 90 0 0>}]
 dyn_pendant_02 [{shapeType=Capsule, length=2, radius=1, shapeName=dyn_pendant_02_capsule, center=<2 0 0>, rotation=<rot 90 0 0>}]
 dyn_pendant_03 [{shapeType=Capsule, length=1.3597, radius=3.8084, shapeName=dyn_pendant_03_capsule, center=<4.3276 0 0>, rotation=<rot 90 0 0>}]
}
constraints [
 {
  bone1Name=base_drip_bot_00
  bone2Name=Melee_TwoHanded_Handle
  swing1LimitDegrees=45
  swing1Motion=Free
  swing2LimitDegrees=45
  swing2Motion=Free
  twistLimitDegrees=45
  twistMotion=Free
 }
 {bone1Name=dyn_drip_bot_00, bone2Name=base_drip_bot_00, swing1LimitDegrees=1, swing2LimitDegrees=1, twistLimitDegrees=1}
 {bone1Name=dyn_drip_bot_01, bone2Name=dyn_drip_bot_00, swing1LimitDegrees=1, swing2LimitDegrees=1, twistLimitDegrees=1}
 {
  bone1Name=base_drip_main_00
  bone2Name=Melee_TwoHanded_Handle
  swing1LimitDegrees=45
  swing1Motion=Free
  swing2LimitDegrees=45
  swing2Motion=Free
  twistLimitDegrees=45
  twistMotion=Free
 }
 {bone1Name=dyn_drip_main_00, bone2Name=base_drip_main_00, swing1LimitDegrees=1.5, swing2LimitDegrees=1.5, twistLimitDegrees=1}
 {bone1Name=dyn_drip_main_01, bone2Name=dyn_drip_main_00, swing1LimitDegrees=1.5, swing2LimitDegrees=1.5, twistLimitDegrees=1}
 {bone1Name=dyn_drip_mid_00, bone2Name=dyn_drip_main_01, swing1LimitDegrees=1, swing2LimitDegrees=1, twistLimitDegrees=1}
 {bone1Name=dyn_drip_mid_01, bone2Name=dyn_drip_mid_00, swing1LimitDegrees=1, swing2LimitDegrees=1, twistLimitDegrees=1}
 {bone1Name=dyn_drip_top_00, bone2Name=dyn_drip_main_01, swing1LimitDegrees=2, swing2LimitDegrees=2, twistLimitDegrees=1}
 {bone1Name=dyn_drip_top_01, bone2Name=dyn_drip_top_00, swing1LimitDegrees=2, swing2LimitDegrees=2, twistLimitDegrees=1}
 {bone1Name=dyn_drip_top_02, bone2Name=dyn_drip_top_01, swing1LimitDegrees=3, swing2LimitDegrees=3, twistLimitDegrees=1}
 {
  bone1Name=base_pendant_00
  bone2Name=Melee_TwoHanded_Handle
  swing1LimitDegrees=45
  swing1Motion=Free
  swing2LimitDegrees=45
  swing2Motion=Free
  twistLimitDegrees=45
  twistMotion=Free
 }
 {bone1Name=dyn_pendant_00, bone2Name=base_pendant_00, swing1LimitDegrees=25, swing2LimitDegrees=20, twistLimitDegrees=1}
 {bone1Name=dyn_pendant_01, bone2Name=dyn_pendant_00, swing1LimitDegrees=25, swing2LimitDegrees=25, twistLimitDegrees=1}
 {bone1Name=dyn_pendant_02, bone2Name=dyn_pendant_01, swing1LimitDegrees=25, swing2LimitDegrees=25, twistLimitDegrees=1}
 {bone1Name=dyn_pendant_03, bone2Name=dyn_pendant_02, swing1LimitDegrees=25, swing2LimitDegrees=25, twistLimitDegrees=1}
]
